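# Physics asset for the backpack mesh part of the female medium QuicheLorraineCrisp_Hair character, covering body bones and dynamic "untucked" front and back chains.
# Capsules approximate the torso, limbs, neck and head, with constraints only along the neck-to-head link and the dynamic chains.
# Both chains hang from the head through free-motion base joints, then use angle-limited joints.
# The seven-bone front chain passes through an intermediate shoulder base, while the back has three larger capsules.
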PhysicsAsset /BRCosmetics/Characters/Player/Female/Medium/Bodies/F_MED_QuicheLorraineCrisp_Hair/Meshes/Parts/F_MED_QuicheLorraineCrisp_Hair_Backpack
bodies {
 calf_l [{shapeType=Capsule, length=47.5, radius=10.5, shapeName=calf_l_capsule, center=<-22.7727 0.0001 0>, rotation=<rot 90 -0.0003 0>}]
 calf_r [{shapeType=Capsule, length=47.5, radius=10.5, shapeName=calf_r_capsule, center=<22.7727 0 0>, rotation=<rot -90 -0.0001 0>}]
 clavicle_l [{shapeType=Capsule, length=12.5, radius=9, shapeName=clavicle_l_capsule, center=<4.8594 0.717 -3.6122>, rotation=<rot -90 -0.0002 0>}]
 clavicle_r [{shapeType=Capsule, length=12, radius=9.5, shapeName=clavicle_r_capsule, center=<-4.8594 0 2.323>, rotation=<rot 90 0.0002 0>}]
 dyn_untucked_back_01 [{shapeType=Capsule, length=11, radius=11.5, shapeName=dyn_untucked_back_01_capsule, center=<10.6994 0 10>, rotation=<rot 0 -90 -89.9995>}]
 dyn_untucked_back_02 [{shapeType=Capsule, length=11, radius=11.5, shapeName=dyn_untucked_back_02_capsule, center=<10.6994 0 10>, rotation=<rot 0 -90 -89.9995>}]
 dyn_untucked_back_03 [{shapeType=Capsule, length=11, radius=11, shapeName=dyn_untucked_back_03_capsule, center=<10.6994 0 10>, rotation=<rot 0 -90 -89.9995>}]
 dyn_untucked_back_base [{shapeType=Capsule, length=0, radius=1, shapeName=dyn_untucked_back_base_capsule}]
 dyn_untucked_front_01 [{shapeType=Capsule, length=5, radius=3, shapeName=dyn_untucked_front_01_capsule, center=<3.8697 0 0>, rotation=<rot -90 -0.0001 0>}]
 dyn_untucked_front_02 [{shapeType=Capsule, length=2.5, radius=2.5, shapeName=dyn_untucked_front_02_capsule, center=<1.8976 -0.4614 0.5>, rotation=<rot -80 0 0>}]
 dyn_untucked_front_03 [{shapeType=Capsule, length=3.5, radius=4, shapeName=dyn_untucked_front_03_capsule, center=<2.5793 0 1.5>, rotation=<rot -90 0 0>}]
 dyn_untucked_front_04 [{shapeType=Capsule, length=1.5, radius=3.5, shapeName=dyn_untucked_front_04_capsule, center=<2.5625 0 0.7305>, rotation=<rot -80 180 180>}]
 dyn_untucked_front_05 [{shapeType=Capsule, length=1.5, radius=3.5, shapeName=dyn_untucked_front_05_capsule, center=<2.5423 0 0.5>, rotation=<rot -90 0 0>}]
 dyn_untucked_front_06 [{shapeType=Capsule, length=2, radius=3, shapeName=dyn_untucked_front_06_capsule, center=<2.5423 0 0.5>, rotation=<rot -90 0 0>}]
 dyn_untucked_front_07 [{shapeType=Capsule, length=5, radius=3, shapeName=dyn_untucked_front_07_capsule, center=<3.5733 0 0.5>, rotation=<rot -90 0 0>}]
 dyn_untucked_front_head_base [{shapeType=Capsule, length=5, radius=3, shapeName=dyn_untucked_front_head_base_capsule}]
 dyn_untucked_front_shoulder_base [{shapeType=Capsule, length=5, radius=3, shapeName=dyn_untucked_front_shoulder_base_capsule, rotation=<rot -90 0 0>}]
 head [{shapeType=Capsule, length=3.5, radius=8, shapeName=head_capsule, center=<1.1764 3.0441 0>, rotation=<rot 90 -0.0002 0>}]
 lowerarm_l [{shapeType=Capsule, length=23, radius=4.5, shapeName=lowerarm_l_capsule, center=<13.761 0 0>, rotation=<rot -90 0 0>}]
 lowerarm_r [{shapeType=Capsule, length=23, radius=4.5, shapeName=lowerarm_r_capsule, center=<-13.761 0 0>, rotation=<rot 90 0.0002 0>}]
 neck_01 [{shapeType=Capsule, length=2.5, radius=5.5, shapeName=neck_01_capsule, center=<0 0.9573 0>, rotation=<rot 90 -0.0001 0>}]
 neck_02 [{shapeType=Capsule, length=3, radius=6, shapeName=neck_02_capsule, center=<0 0.9573 0>, rotation=<rot 90 -0.0001 0>}]
 pelvis [{shapeType=Capsule, length=11.5, radius=13, shapeName=pelvis_capsule, center=<-5.6643 -2.253 0>}]
 spine_01 [{shapeType=Capsule, length=21.5, radius=9.5, shapeName=spine_01_capsule}]
 spine_02 [{shapeType=Capsule, length=13.5, radius=10, shapeName=spine_02_capsule, center=<0 1.407 0>}]
 spine_03 [{shapeType=Capsule, length=14.5, radius=10, shapeName=spine_03_capsule, center=<0 1.989 0>}]
 spine_04 [{shapeType=Capsule, length=11, radius=12, shapeName=spine_04_capsule, center=<0 1.3178 0>}]
 spine_05 [{shapeType=Capsule, length=23.5, radius=9.5, shapeName=spine_05_capsule, center=<0.5269 -0.5262 0>}]
 thigh_l [{shapeType=Capsule, length=29, radius=10, shapeName=thigh_l_capsule, center=<-18.589 0 0>, rotation=<rot 90 -0.0001 0>}]
 thigh_r [{shapeType=Capsule, length=29, radius=10, shapeName=thigh_r_capsule, center=<18.589 0 0>, rotation=<rot -90 -0.0001 0>}]
 upperarm_l [{shapeType=Capsule, length=20.5, radius=5, shapeName=upperarm_l_capsule, center=<10.8208 0 0>, rotation=<rot -90 0 0>}]
 upperarm_r [{shapeType=Capsule, length=20, radius=5.5, shapeName=upperarm_r_capsule, center=<-10.8208 0.9971 -0.8153>, rotation=<rot 90 0.0002 0>}]
}
constraints [
 {
  bone1Name=neck_01
  bone2Name=spine_05
  swing1LimitDegrees=45
  swing1Motion=Free
  swing2LimitDegrees=45
  swing2Motion=Free
  twistLimitDegrees=45
  twistMotion=Free
 }
 {
  bone1Name=neck_02
  bone2Name=neck_01
  swing1LimitDegrees=45
  swing1Motion=Free
  swing2LimitDegrees=45
  swing2Motion=Free
  twistLimitDegrees=45
  twistMotion=Free
 }
 {
  bone1Name=head
  bone2Name=neck_02
  swing1LimitDegrees=45
  swing1Motion=Free
  swing2LimitDegrees=45
  swing2Motion=Free
  twistLimitDegrees=45
  twistMotion=Free
 }
 {bone1Name=dyn_untucked_back_01, bone2Name=dyn_untucked_back_base, swing1LimitDegrees=20, swing2LimitDegrees=60, twistLimitDegrees=10}
 {bone1Name=dyn_untucked_back_02, bone2Name=dyn_untucked_back_01, swing1LimitDegrees=20, swing2LimitDegrees=60, twistLimitDegrees=10}
 {bone1Name=dyn_untucked_back_03, bone2Name=dyn_untucked_back_02, swing1LimitDegrees=20, swing2LimitDegrees=60, twistLimitDegrees=10}
 {
  bone1Name=dyn_untucked_front_head_base
  bone2Name=head
  swing1LimitDegrees=45
  swing1Motion=Free
  swing2LimitDegrees=45
  swing2Motion=Free
  twistLimitDegrees=45
  twistMotion=Free
 }
 {
  bone1Name=dyn_untucked_front_01
  bone2Name=dyn_untucked_front_head_base
  swing1LimitDegrees=45
  swing1Motion=Free
  swing2LimitDegrees=45
  swing2Motion=Free
  twistLimitDegrees=45
  twistMotion=Free
 }
 {
  bone1Name=dyn_untucked_front_shoulder_base
  bone2Name=dyn_untucked_front_01
  swing1LimitDegrees=45
  swing1Motion=Free
  swing2LimitDegrees=45
  swing2Motion=Free
  twistLimitDegrees=45
  twistMotion=Free
 }
 {
  bone1Name=dyn_untucked_front_02
  bone2Name=dyn_untucked_front_shoulder_base
  swing1LimitDegrees=80
  swing2LimitDegrees=60
  twistLimitDegrees=10
 }
 {bone1Name=dyn_untucked_front_03, bone2Name=dyn_untucked_front_02, swing1LimitDegrees=60, swing2LimitDegrees=60, twistLimitDegrees=10}
 {bone1Name=dyn_untucked_front_04, bone2Name=dyn_untucked_front_03, swing1LimitDegrees=60, swing2LimitDegrees=60, twistLimitDegrees=10}
 {bone1Name=dyn_untucked_front_05, bone2Name=dyn_untucked_front_04, swing1LimitDegrees=60, swing2LimitDegrees=60, twistLimitDegrees=10}
 {bone1Name=dyn_untucked_front_06, bone2Name=dyn_untucked_front_05, swing1LimitDegrees=60, swing2LimitDegrees=60, twistLimitDegrees=10}
 {bone1Name=dyn_untucked_front_07, bone2Name=dyn_untucked_front_06, swing1LimitDegrees=60, swing2LimitDegrees=60, twistLimitDegrees=10}
 {
  bone1Name=dyn_untucked_back_base
  bone2Name=head
  swing1LimitDegrees=45
  swing1Motion=Free
  swing2LimitDegrees=45
  swing2Motion=Free
  twistLimitDegrees=45
  twistMotion=Free
 }
]
