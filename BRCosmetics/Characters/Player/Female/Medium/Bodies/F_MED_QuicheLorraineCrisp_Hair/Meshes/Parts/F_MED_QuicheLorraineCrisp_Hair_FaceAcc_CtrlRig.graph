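# Control Rig for the QuicheLorraineCrisp hair face accessory on a medium female character, applying corrective offsets to its dyn_untucked bones as the head turns.
# Two spherical pose readers measure head orientation relative to spine_05, with debug drawing enabled, and fill Look Left and Look Right.
# Those weights then drive Modify Transforms on the back base, plus the front base for Look Right.
# Two parent constraints on dyn_untucked_back_04 sit outside the execution chain.
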ControlRigBlueprint /BRCosmetics/Characters/Player/Female/Medium/Bodies/F_MED_QuicheLorraineCrisp_Hair/Meshes/Parts/F_MED_QuicheLorraineCrisp_Hair_FaceAcc_CtrlRig
graphs [
 RigVMFunctionLibrary {links [], nodes []}
 RigVMFunction_Sequence_1_ContainedGraph {nodes [
   n1 = RigVMFunctionEntryNode()
   RigVMFunction_Sequence = &RigVMFunction_Sequence(ExecuteContext=@n1.ExecuteContext)
   n2 = &RigVMFunction_Sequence(ExecuteContext=@RigVMFunction_Sequence.B)
   RigVMFunctionReturnNode(A=@RigVMFunction_Sequence.A, B=@n2.A, C=@n2.B)
  ]}
 RigVMModel {nodes [
   &RigUnit_ParentConstraint(
    AdvancedSettings=(InterpolationType=Average,RotationOrderForFilter=XZY)
    Child=(Type=Bone,Name="dyn_untucked_back_04")
    ExecutePin=()
    Parents=((Item=(Type=Bone,Name="head"),Weight=0.6),(Item=(Type=Bone,Name="spine_05"),Weight=0.4))
    Weight=RigVMVariableNode() {variable=Look Left}.Value
   )
   > &RigUnit_ParentConstraint(
    AdvancedSettings=(InterpolationType=Average,RotationOrderForFilter=XZY)
    Child=(Type=Bone,Name="dyn_untucked_back_04")
    Parents=((Item=(Type=Bone,Name="head"),Weight=0.6),(Item=(Type=Bone,Name="spine_05"),Weight=0.4))
    Weight=1
    Parents.1.Weight=RigVMVariableNode() {variable=Look Right}.Value
   )
   RigUnit_BeginExecution = &RigUnit_BeginExecution()
   RigVMFunction_Sequence = RigVMAggregateNode(ExecuteContext=@RigUnit_BeginExecution.ExecutePin)
   n1 = &RigUnit_SphericalPoseReader(
    ActiveRegionScaleFactors=(PositiveWidth=0.5,NegativeWidth=0.5,PositiveHeight=0.5,NegativeHeight=0.5)
    ActiveRegionSize=0.1
    Debug=(bDrawDebug=true,bDraw2D=False,bDrawLocalAxes=False,DebugScale=25,DebugSegments=20,DebugThickness=0.25)
    DriverAxis=(X=0,Y=1,Z=0)
    DriverItem=(Type=Bone,Name="head")
    FalloffRegionScaleFactors=(PositiveWidth=1,NegativeWidth=1,PositiveHeight=1.5,NegativeHeight=1)
    FalloffSize=0.22
    FlipHeightScaling=False
    FlipWidthScaling=False
    OptionalParentItem=(Type=Bone,Name="spine_05")
    RotationOffset=(X=14.5196,Y=0,Z=0)
    ExecutePin=@RigVMFunction_Sequence.A
   )
   > n2 = &RigUnit_SphericalPoseReader(
    ActiveRegionScaleFactors=(PositiveWidth=0.5,NegativeWidth=0.5,PositiveHeight=0.5,NegativeHeight=0.5)
    ActiveRegionSize=0.1
    Debug=(bDrawDebug=true,bDraw2D=False,bDrawLocalAxes=False,DebugScale=25,DebugSegments=20,DebugThickness=0.25)
    DriverAxis=(X=0,Y=1,Z=0)
    DriverItem=(Type=Bone,Name="head")
    FalloffRegionScaleFactors=(PositiveWidth=1,NegativeWidth=1,PositiveHeight=1,NegativeHeight=2.8)
    FalloffSize=0.22
    FlipHeightScaling=False
    FlipWidthScaling=False
    OptionalParentItem=(Type=Bone,Name="spine_05")
    RotationOffset=(X=166.225,Y=-0.6701,Z=6.335)
   )
   RigVMVariableNode(ExecuteContext=@RigVMFunction_Sequence.B, Value=@n1.OutputParam) {variable=Look Right}
   > RigVMVariableNode(Value=@n2.OutputParam) {variable=Look Left}
   &RigUnit_ModifyTransforms(
    ItemToModify=((Item=(Type=Bone,Name="dyn_untucked_back_base"),Transform=(Rotation=(X=0.1105,Y=-0.0818,Z=0.5666,W=0.8125),Translation=(X=2.363,Y=3.0322,Z=0.6108),Scale3D=(X=1,Y=1,Z=1))))
    WeightMinimum=0
    ExecutePin=@RigVMFunction_Sequence.C
    Weight=RigVMVariableNode() {variable=Look Left}.Value
   )
   > &RigUnit_ModifyTransforms(
    ItemToModify="((Item=(Type=Bone,Name=\"dyn_untucked_back_base\"),Transform=(Rotation=(X=0.143,Y=0.0637,Z=0.5307,W=-0.833),Translation=(X=2.1979,Y=-0.8544,Z=1.3817),Scale3D=(X=1,Y=1,Z=1))),(Item=(Type=Bone,Name=\"dyn_untucked_front_head_base\"),Transform=(Rotation=(X=0,Y=0,Z ...(truncated)"
    WeightMinimum=0
    Weight=RigVMVariableNode() {variable=Look Right}.Value
   )
  ]}
]
hierarchy [
 root {loc=<0 0 0>, rot=<0 0 0 1>}
 pelvis {parent=root, loc=<0 -0.5301 98.0571>, rot=<0.0127 0.707 -0.0127 -0.707>}
 spine_01 {parent=pelvis, loc=<6.1208 -0.9328 0>, rot=<0 0 0.034 -0.9994>}
 spine_02 {parent=spine_01, loc=<6.5997 -0.2095 0>, rot=<0 0 0.0238 -0.9997>}
 spine_03 {parent=spine_02, loc=<6.4914 -0.1237 0>, rot=<0 0 0.0149 -0.9999>}
 spine_04 {parent=spine_03, loc=<6.9598 0.0794 0>, rot=<0 0 -0.0132 -0.9999>}
 spine_05 {parent=spine_04, loc=<6.9987 0 0>, rot=<0 0 -0.0015 -1>}
 neck_01 {parent=spine_05, loc=<11.3138 0.1529 0>, rot=<0 0 -0.2003 -0.9797>}
 neck_02 {parent=neck_01, loc=<4.5627 0 0>, rot=<0 0 0.0504 -0.9987>}
 head {parent=neck_02, loc=<5.3914 0 0>, rot=<0 0 0.115 -0.9934>}
 dyn_untucked_front_head_base {parent=head, loc=<-0.7653 3.2378 6.7585>, rot=<0.5194 0.4799 0.5233 0.4756>}
 dyn_untucked_front_01 {parent=dyn_untucked_front_head_base, loc=<0 0 0>, rot=<0.0599 -0.7067 -0.01 -0.7049>}
 dyn_untucked_front_shoulder_base {parent=dyn_untucked_front_01, loc=<8.9231 0.8841 0.4476>, rot=<-0.0106 0.2776 -0.0095 -0.9606>}
 dyn_untucked_front_02 {parent=dyn_untucked_front_shoulder_base, loc=<0 0 0>, rot=<0 0 0 1>}
 dyn_untucked_front_03 {parent=dyn_untucked_front_02, loc=<7.7584 -0.0731 0.4155>, rot=<-0.0124 0.0469 -0.0297 -0.9984>}
 dyn_untucked_front_04 {parent=dyn_untucked_front_03, loc=<8.0922 0.1758 -0.6931>, rot=<-0.023 -0.1058 0.0444 -0.9931>}
 dyn_untucked_front_05 {parent=dyn_untucked_front_04, loc=<5.1987 0.1748 -1.0415>, rot=<0.0169 -0.2108 -0.0253 -0.9771>}
 dyn_untucked_front_06 {parent=dyn_untucked_front_05, loc=<8.7647 -0.1347 0.1577>, rot=<-0.0177 -0.0282 0.0294 -0.999>}
 dyn_untucked_front_07 {parent=dyn_untucked_front_06, loc=<6.6716 -0.082 0.2917>, rot=<0 0 0 1>}
 dyn_untucked_back_base {parent=head, loc=<9.369 -4.0777 0>, rot=<0.5021 -0.4979 0.4979 -0.5021>}
 dyn_untucked_back_01 {parent=dyn_untucked_back_base, loc=<0 0 0>, rot=<0 -0.5736 0 -0.8192>}
 dyn_untucked_back_02 {parent=dyn_untucked_back_01, loc=<6.0958 0 0.1664>, rot=<0 -0.0813 0 -0.9967>}
 dyn_untucked_back_03 {parent=dyn_untucked_back_02, loc=<7.5172 0 0>, rot=<0 0.0759 0 -0.9971>}
 dyn_untucked_back_04 {parent=dyn_untucked_back_03, loc=<7.3685 0 0>, rot=<0 -0.0454 0 -0.999>}
 dyn_untucked_back_05 {parent=dyn_untucked_back_04, loc=<8.1513 0 0>, rot=<0 -0.0852 0 -0.9964>}
 dyn_untucked_back_06 {parent=dyn_untucked_back_05, loc=<10.3487 0 0>, rot=<0 -0.0219 0 -0.9998>}
 dyn_untucked_back_07 {parent=dyn_untucked_back_06, loc=<8.7479 0 0>, rot=<0 -0.0127 0 -0.9999>}
 dyn_untucked_back_08 {parent=dyn_untucked_back_07, loc=<9.1963 0 0>, rot=<0 0 0 1>}
 dyn_hair_r_base {parent=head, loc=<1.6801 -1.1809 7.537>, rot=<0.1174 -0.6658 0.7324 -0.0804>}
 dyn_hair_r_01 {parent=dyn_hair_r_base, loc=<0 0 0>, rot=<0 0 0 1>}
 dyn_hair_r_02 {parent=dyn_hair_r_01, loc=<7.3013 0 0>, rot=<0 0.2797 -0.0749 -0.9572>}
 dyn_hair_r_03 {parent=dyn_hair_r_02, loc=<7.2681 0 0>, rot=<0 0 0 1>}
 dyn_hair_l_base {parent=head, loc=<1.6801 -1.1809 -7.5373>, rot=<0.0872 -0.7704 0.6215 -0.1124>}
 dyn_hair_l_01 {parent=dyn_hair_l_base, loc=<0 0 0>, rot=<0 0 0 1>}
 dyn_hair_l_02 {parent=dyn_hair_l_01, loc=<7.3013 0 0>, rot=<0 0.2757 0.0739 -0.9584>}
 dyn_hair_l_03 {parent=dyn_hair_l_02, loc=<7.2633 0 0>, rot=<0 0 0 1>}
 jnt_support_hair_l {parent=head, loc=<4.3032 6.8969 -5.5052>, rot=<0.0071 -0.7753 0.6066 -0.1761>}
 clavicle_l {parent=spine_05, loc=<5.4831 -0.1053 -3.2946>, rot=<0.0465 -0.7289 -0.009 -0.6829>}
 upperarm_l {parent=clavicle_l, loc=<9.8658 -1.2136 -0.4105>, rot=<0.0164 -0.3245 -0.0269 -0.9454>}
 lowerarm_l {parent=upperarm_l, loc=<26.465 -0.8345 0.259>, rot=<0.0036 0.0161 -0.2298 -0.9731>}
 hand_l {parent=lowerarm_l, loc=<22.7355 -0.3203 -0.6105>, rot=<-0.6784 -0.0223 0.0735 -0.7307>}
 upperarm_twist_01_l {parent=upperarm_l, loc=<7.2135 -0.1134 0>, rot=<0 0 0 1>}
 upperarm_twist_02_l {parent=upperarm_l, loc=<19.7998 -0.1134 0>, rot=<0 0 0 1>}
 clavicle_r {parent=spine_05, loc=<5.4831 -0.1053 3.2946>, rot=<0.7289 0.0465 -0.6829 0.009>}
 upperarm_r {parent=clavicle_r, loc=<-9.8658 1.2136 0.4105>, rot=<0.0164 -0.3245 -0.0269 -0.9454>}
 lowerarm_r {parent=upperarm_r, loc=<-26.465 0.8345 -0.259>, rot=<0.0036 0.0161 -0.2298 -0.9731>}
 hand_r {parent=lowerarm_r, loc=<-22.7355 0.3203 0.6106>, rot=<-0.6784 -0.0223 0.0735 -0.7307>}
 upperarm_twist_01_r {parent=upperarm_r, loc=<-7.2135 0.1134 0>, rot=<0 0 0 -1>}
 thigh_l {parent=pelvis, loc=<-1.0855 -0.2109 -8.7053>, rot=<0.0009 0.0355 0.0155 -0.9993>}
 calf_l {parent=thigh_l, loc=<-42.5284 -0.0019 -0.7297>, rot=<0.0006 -0.0161 -0.0391 -0.9991>}
 foot_l {parent=calf_l, loc=<-46.7401 -2.5014 -0.592>, rot=<0.0427 -0.0238 0.0353 -0.9982>}
 thigh_twist_01_l {parent=thigh_l, loc=<-27.952 0 0>, rot=<0.0013 0 0 -1>}
 thigh_r {parent=pelvis, loc=<-1.0855 -0.2109 8.7053>, rot=<0.0355 -0.0009 0.9993 0.0155>}
 calf_r {parent=thigh_r, loc=<42.5284 0.0019 0.7297>, rot=<0.0006 -0.0161 -0.0391 -0.9991>}
 foot_r {parent=calf_r, loc=<46.7401 2.5014 0.592>, rot=<0.0427 -0.0238 0.0353 -0.9982>}
 thigh_twist_01_r {parent=thigh_r, loc=<27.952 0 0>, rot=<0.0013 0 0 -1>}
]
variables [
 Look Left {type=double, default=0}
 Look Right {type=double, default=0}
]
