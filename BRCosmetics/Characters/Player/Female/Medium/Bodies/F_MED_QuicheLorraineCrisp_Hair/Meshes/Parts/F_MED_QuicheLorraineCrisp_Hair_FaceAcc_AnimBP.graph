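# Post-body AnimBP for the QuicheLorraineCrisp female hair/face-accessory part, animating dynamic hair and "untucked" bones.
# Copied body pose passes through Control Rig into two bool-selected RigidBody sims (standard, FE physics), then front constraints.
# That cached pose feeds a backpack-asset RigidBody, followed by weighted bone constraints and additive corrective offsets.
# EventGraph updates gravity override and a lime-emote curve value, and sets backbling, cape-skydiving and front-end idle variables.
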AnimBlueprint /BRCosmetics/Characters/Player/Female/Medium/Bodies/F_MED_QuicheLorraineCrisp_Hair/Meshes/Parts/F_MED_QuicheLorraineCrisp_Hair_FaceAcc_AnimBP
target_skeleton=&/F_MED_QuicheLorraineCrisp_Hair_FaceAcc_Skeleton
graphs [
 AnimGraph {nodes [
   "Constraint - dyn_untucked_front_03 (P)"!(ComponentPose, bAlphaBoolEnabled:bool=True, ConstraintWeights_0=1) -> (Pose) {
    alphaInputType=Bool
    alphaBoolBlend {blendInTime=0.2, blendOutTime=0.2}
    boneToModify {boneName=dyn_untucked_front_03}
    constraintSetup [{targetBone {boneName=spine_03}}]
    constraintWeights [1]
   }
   n1 = RigidBody!(ComponentPose=Local To Component(LocalPose=Use cached pose 'dyn cache'()), Alpha, OverrideWorldGravity:Vector="0,0,-980", ExternalForce) -> (Pose) {
    bClampLinearTranslationLimitToRefPose=true
    bOverrideWorldGravity=true
    overridePhysicsAsset=&/F_MED_QuicheLorraineCrisp_Hair_Backpack
    simulationSpace=BaseBoneSpace
    baseBoneRef {boneName=spine_05}
    componentAppliedLinearAccClamp=<3000 3000 3000>
    overrideWorldGravity {z=-980}
    simSpaceSettings {
     dampingAlpha=0
     maxAngularAcceleration=40
     maxAngularVelocity=8
     maxLinearAcceleration=1500
     maxLinearVelocity=700
     worldAlpha=1
     externalLinearDragV=<0.2 0.2 0.2>
    }
   }
   n2 = "Constraint - dyn_untucked_back_01 (P)"(ComponentPose=@n1, Alpha=1, ConstraintWeights_0=1) {boneToModify {boneName=dyn_untucked_back_01}, constraintSetup [{targetBone {boneName=head}}], constraintWeights [1]}
   n3 = "Constraint - dyn_untucked_back_03 (PPPPPP)"!(
    ComponentPose=@n2
    Alpha=1
    ConstraintWeights_0=0.0031
    ConstraintWeights_1:float=0.0233
    ConstraintWeights_2:float=0.001
    ConstraintWeights_3:float=0.3353
    ConstraintWeights_4:float=0.2305
    ConstraintWeights_5:float=0.4068
   ) -> (Pose) {
    boneToModify {boneName=dyn_untucked_back_03}
    constraintSetup [
     {targetBone {boneName=spine_04}}
     {targetBone {boneName=spine_05}}
     {targetBone {boneName=clavicle_r}}
     {targetBone {boneName=neck_01}}
     {targetBone {boneName=neck_02}}
     {targetBone {boneName=head}}
    ]
    constraintWeights [1,1,1,1,1,1]
   }
   n4 = "Constraint - dyn_untucked_back_04 (PPP)"!(ComponentPose=@n3, Alpha=1, ConstraintWeights_0=0.8968, ConstraintWeights_1:float=0.051, ConstraintWeights_2:float=0.0522) -> (Pose) {
    boneToModify {boneName=dyn_untucked_back_04}
    constraintSetup [{targetBone {boneName=spine_05}}, {targetBone {boneName=clavicle_l}}, {targetBone {boneName=clavicle_r}}]
    constraintWeights [1,1,1]
   }
   n5 = "Constraint - dyn_untucked_back_05 (PPPP)"!(
    ComponentPose=@n4
    Alpha=1
    ConstraintWeights_0=0.0709
    ConstraintWeights_1:float=0.6618
    ConstraintWeights_2:float=0.1274
    ConstraintWeights_3:float=0.1399
   ) -> (Pose) {
    boneToModify {boneName=dyn_untucked_back_05}
    constraintSetup [{targetBone {boneName=spine_04}}, {targetBone {boneName=spine_05}}, {targetBone {boneName=clavicle_l}}, {targetBone {boneName=clavicle_r}}]
    constraintWeights [1,1,1,1]
   }
   n6 = "Constraint - dyn_untucked_back_06 (PPPPPPP)"!(
    ComponentPose=@n5
    Alpha=1
    ConstraintWeights_0=0.09
    ConstraintWeights_1:float=0.505
    ConstraintWeights_2:float=0.1292
    ConstraintWeights_3:float=0.1482
    ConstraintWeights_4:float=0.0472
    ConstraintWeights_5:float=0.0455
    ConstraintWeights_6:float=0.035
   ) -> (Pose) {
    boneToModify {boneName=dyn_untucked_back_06}
    constraintSetup [
     {targetBone {boneName=spine_02}}
     {targetBone {boneName=spine_04}}
     {targetBone {boneName=spine_03}}
     {targetBone {boneName=spine_05}}
     {targetBone {boneName=clavicle_l}}
     {targetBone {boneName=clavicle_r}}
     {targetBone {boneName=spine_01}}
    ]
    constraintWeights [1,1,1,1,1,1,1]
   }
   n7 = "Constraint - dyn_untucked_back_07 (PPPPP)"!(
    ComponentPose=@n6
    Alpha=1
    ConstraintWeights_0=0.4095
    ConstraintWeights_1:float=0.0399
    ConstraintWeights_2:float=0.2318
    ConstraintWeights_3:float=0.2122
    ConstraintWeights_4:float=0.1066
   ) -> (Pose) {
    boneToModify {boneName=dyn_untucked_back_07}
    constraintSetup [
     {targetBone {boneName=spine_02}}
     {targetBone {boneName=spine_04}}
     {targetBone {boneName=spine_03}}
     {targetBone {boneName=spine_01}}
     {targetBone {boneName=pelvis}}
    ]
    constraintWeights [1,1,1,1,1]
   }
   n8 = "Constraint - dyn_untucked_back_08 (PPP)"!(ComponentPose=@n7, Alpha=1, ConstraintWeights_0=0.0984, ConstraintWeights_1:float=0.1618, ConstraintWeights_2:float=0.7397) -> (Pose) {
    boneToModify {boneName=dyn_untucked_back_08}
    constraintSetup [{targetBone {boneName=spine_02}}, {targetBone {boneName=spine_01}}, {targetBone {boneName=pelvis}}]
    constraintWeights [1,1,1]
   }
   n9 = "Constraint - dyn_hair_r_02 (PPP)"!(ComponentPose=@n8, Alpha=1, ConstraintWeights_0=0.7075, ConstraintWeights_1:float=0.289, ConstraintWeights_2:float=0.0034) -> (Pose) {
    boneToModify {boneName=dyn_hair_r_02}
    constraintSetup [{targetBone {boneName=spine_05}}, {targetBone {boneName=clavicle_r}}, {targetBone {boneName=neck_02}}]
    constraintWeights [1,1,1]
   }
   n10 = "Constraint - dyn_hair_r_01 (P)"(ComponentPose=@n9, Alpha=1, ConstraintWeights_0=1) {boneToModify {boneName=dyn_hair_r_01}, constraintSetup [{targetBone {boneName=head}}], constraintWeights [1]}
   n11 = "Constraint - dyn_hair_r_03 (PPPP)"!(
    ComponentPose=@n10
    Alpha=1
    ConstraintWeights_0=0.0021
    ConstraintWeights_1:float=0.3966
    ConstraintWeights_2:float=0.5743
    ConstraintWeights_3:float=0.027
   ) -> (Pose) {
    boneToModify {boneName=dyn_hair_r_03}
    constraintSetup [
     {targetBone {boneName=spine_04}}
     {targetBone {boneName=spine_05}}
     {targetBone {boneName=clavicle_r}}
     {targetBone {boneName=upperarm_twist_01_r}}
    ]
    constraintWeights [1,1,1,1]
   }
   n12 = "Constraint - dyn_hair_l_01 (P)"(ComponentPose=@n11, Alpha=1, ConstraintWeights_0=1) {boneToModify {boneName=dyn_hair_l_01}, constraintSetup [{targetBone {boneName=head}}], constraintWeights [1]}
   n13 = "Constraint - dyn_hair_l_02 (PPPP)"!(
    ComponentPose=@n12
    Alpha=1
    ConstraintWeights_0=0.91
    ConstraintWeights_1:float=0.0572
    ConstraintWeights_2:float=0.0109
    ConstraintWeights_3:float=0.0219
   ) -> (Pose) {
    boneToModify {boneName=dyn_hair_l_02}
    constraintSetup [{targetBone {boneName=spine_05}}, {targetBone {boneName=clavicle_l}}, {targetBone {boneName=neck_01}}, {targetBone {boneName=neck_02}}]
    constraintWeights [1,1,1,1]
   }
   n14 = "Constraint - dyn_hair_l_03 (PPP)"!(ComponentPose=@n13, Alpha=1, ConstraintWeights_0=0.0377, ConstraintWeights_1:float=0.271, ConstraintWeights_2:float=0.6914) -> (Pose) {
    boneToModify {boneName=dyn_hair_l_03}
    constraintSetup [{targetBone {boneName=upperarm_twist_01_l}}, {targetBone {boneName=spine_05}}, {targetBone {boneName=clavicle_l}}]
    constraintWeights [1,1,1]
   }
   n15 = "Transform (Modify) Bone - Bone: dyn_hair_r_base"(ComponentPose=@n14, Alpha=1, Translation="-1.465,-0.651,2.191", Rotation="0,0,0", Scale="1,1,1") {
    rotationMode=BMM_Additive
    rotationSpace=BCS_BoneSpace
    translationMode=BMM_Additive
    translationSpace=BCS_BoneSpace
    boneToModify {boneName=dyn_hair_r_base}
    translation=<-1.465 -0.651 2.191>
   }
   n16 = "Transform (Modify) Bone - Bone: dyn_hair_l_base"(ComponentPose=@n15, Alpha=1, Translation="-0.355,0.837,3.928", Rotation="0,0,0", Scale="0.8,1,1") {
    rotationMode=BMM_Additive
    rotationSpace=BCS_BoneSpace
    scaleMode=BMM_Replace
    scaleSpace=BCS_BoneSpace
    translationMode=BMM_Additive
    translationSpace=BCS_BoneSpace
    boneToModify {boneName=dyn_hair_l_base}
    scale {x=0.8}
    translation=<-0.355 0.837 3.928>
   }
   n17 = "Transform (Modify) Bone - Bone: dyn_untucked_back_03"(ComponentPose=@n16, Alpha=1, Translation="-2.234,0,3.814", Rotation="0,0,0", Scale="1,1,1") {
    rotationMode=BMM_Additive
    rotationSpace=BCS_BoneSpace
    translationMode=BMM_Additive
    translationSpace=BCS_BoneSpace
    boneToModify {boneName=dyn_untucked_back_03}
    translation {x=-2.234, z=3.814}
   }
   n18 = "Transform (Modify) Bone - Bone: dyn_untucked_back_04"(ComponentPose=@n17, Alpha=1, Translation="0.003,0,-2.286", Rotation="0,0,0", Scale="1,1,1") {
    rotationMode=BMM_Additive
    rotationSpace=BCS_BoneSpace
    translationMode=BMM_Additive
    translationSpace=BCS_BoneSpace
    boneToModify {boneName=dyn_untucked_back_04}
    translation {x=0.003, z=-2.286}
   }
   n19 = "Transform (Modify) Bone - Bone: dyn_untucked_back_05"(ComponentPose=@n18, Alpha=1, Translation="-0.005,0,-1.451", Rotation="0,0,0", Scale="1,1,1") {
    rotationMode=BMM_Additive
    rotationSpace=BCS_BoneSpace
    translationMode=BMM_Additive
    translationSpace=BCS_BoneSpace
    boneToModify {boneName=dyn_untucked_back_05}
    translation {x=-0.005, z=-1.451}
   }
   n20 = "Transform (Modify) Bone - Bone: dyn_untucked_back_03"(ComponentPose=@n19, Alpha=1, Translation="0,0,3.814", Rotation="-0.000,-5.000,0.000", Scale="1,1,1") {
    rotationMode=BMM_Additive
    rotationSpace=BCS_BoneSpace
    translationMode=BMM_Additive
    translationSpace=BCS_BoneSpace
    boneToModify {boneName=dyn_untucked_back_03}
    rotation=<rot 0 -5 0>
    translation {z=3.814}
   }
   n21 = "Transform (Modify) Bone - Bone: jnt_support_hair_l"(ComponentPose=@n20, Alpha=1, Translation="0,0.8,0", Rotation="0,0,0", Scale="1,1,1") {translationMode=BMM_Additive, translationSpace=BCS_BoneSpace, boneToModify {boneName=jnt_support_hair_l}, translation {y=0.8}}
   n22 = "Constraint - dyn_hair_l_02 (PPPP)"(ComponentPose=@n21, Alpha=1, ConstraintWeights_0=0.8677, ConstraintWeights_1=0.1091, ConstraintWeights_2=0.0179, ConstraintWeights_3=0.0053) {
    boneToModify {boneName=dyn_hair_l_02}
    constraintSetup [{targetBone {boneName=spine_05}}, {targetBone {boneName=clavicle_l}}, {targetBone {boneName=neck_02}}, {targetBone {boneName=neck_01}}]
    constraintWeights [1,1,1,1]
   }
   n23 = "Transform (Modify) Bone - Bone: dyn_hair_l_02"(ComponentPose=@n22, Alpha=1, Translation="0,0,2", Rotation="10,0,0", Scale="1,1,1") {
    rotationMode=BMM_Additive
    rotationSpace=BCS_BoneSpace
    translationMode=BMM_Additive
    translationSpace=BCS_BoneSpace
    boneToModify {boneName=dyn_hair_l_02}
    rotation {pitch=10}
    translation {z=2}
   }
   n24 = "Transform (Modify) Bone - Bone: dyn_untucked_back_03"!(ComponentPose=@n23, bAlphaBoolEnabled:bool=True, Translation="0,0,0", Rotation="-2.5,0,0", Scale="1,1,1") -> (Pose) {
    alphaInputType=Bool
    rotationMode=BMM_Additive
    rotationSpace=BCS_BoneSpace
    scaleMode=BMM_Additive
    scaleSpace=BCS_BoneSpace
    translationMode=BMM_Additive
    translationSpace=BCS_BoneSpace
    alphaBoolBlend {blendInTime=0.2, blendOutTime=0.2}
    boneToModify {boneName=dyn_untucked_back_03}
    rotation {pitch=-2.5}
   }
   Output Pose(Result=Component To Local(ComponentPose=@n24)) {layerGroup=None}
   n25 = RigidBody(ComponentPose=Local To Component(LocalPose=Use cached pose 'CtrlRig_Cache'()), OverrideWorldGravity="0,0,-980") {
    bClampLinearTranslationLimitToRefPose=true
    bOverrideWorldGravity=true
    overridePhysicsAsset=&/F_MED_QuicheLorraineCrisp_Hair_FaceAcc_Physics
    simulationSpace=BaseBoneSpace
    baseBoneRef {boneName=spine_05}
    componentAppliedLinearAccClamp=<3000 3000 3000>
    overrideWorldGravity {z=-980}
    simSpaceSettings {
     dampingAlpha=0
     maxAngularAcceleration=2
     maxAngularVelocity=2
     maxLinearAcceleration=2
     maxLinearVelocity=2
     worldAlpha=0.2
     externalLinearDragV=<0.2 0.2 0.2>
    }
   }
   n26 = Component To Local(ComponentPose=@n25)
   n27 = RigidBody(ComponentPose=Local To Component(LocalPose=Use cached pose 'CtrlRig_Cache'()), OverrideWorldGravity="0,0,-980") {
    bClampLinearTranslationLimitToRefPose=true
    bOverrideWorldGravity=true
    overridePhysicsAsset=&/F_MED_QuicheLorraineCrisp_Hair_FaceAcc_FEPh
    simulationSpace=BaseBoneSpace
    baseBoneRef {boneName=spine_05}
    componentAppliedLinearAccClamp=<3000 3000 3000>
    overrideWorldGravity {z=-980}
    simSpaceSettings {
     dampingAlpha=0
     maxAngularAcceleration=2
     maxAngularVelocity=2
     maxLinearAcceleration=2
     maxLinearVelocity=2
     worldAlpha=0.2
     externalLinearDragV=<0.2 0.2 0.2>
    }
   }
   n28 = Component To Local(ComponentPose=@n27)
   n29 = "Constraint - dyn_untucked_front_05 (P)"!(
    ComponentPose=Local To Component(LocalPose=Blend Poses by bool(BlendPose_0=@n28, BlendPose_1=@n26))
    Alpha=1
    ConstraintWeights_0=1
    ConstraintWeights_1:float=0
   ) -> (Pose) {boneToModify {boneName=dyn_untucked_front_05}, constraintSetup [{targetBone {boneName=spine_04}}], constraintWeights [1,0]}
   n30 = "Constraint - dyn_untucked_front_06 (PPPPP)"!(
    ComponentPose=@n29
    Alpha=1
    ConstraintWeights_0=0.3038
    ConstraintWeights_1:float=0.3543
    ConstraintWeights_2:float=0.1938
    ConstraintWeights_3:float=0.1397
    ConstraintWeights_4:float=0.0085
   ) -> (Pose) {
    boneToModify {boneName=dyn_untucked_front_06}
    constraintSetup [
     {targetBone {boneName=spine_02}}
     {targetBone {boneName=spine_04}}
     {targetBone {boneName=spine_03}}
     {targetBone {boneName=spine_01}}
     {targetBone {boneName=pelvis}}
    ]
    constraintWeights [1,1,1,1,1]
   }
   n31 = "Constraint - dyn_untucked_front_07 (PPPPP)"!(
    ComponentPose=@n30
    Alpha=1
    ConstraintWeights_0=0.3057
    ConstraintWeights_1:float=0.0784
    ConstraintWeights_2:float=0.0102
    ConstraintWeights_3:float=0.1728
    ConstraintWeights_4:float=0.433
   ) -> (Pose) {
    boneToModify {boneName=dyn_untucked_front_07}
    constraintSetup [
     {targetBone {boneName=spine_02}}
     {targetBone {boneName=spine_04}}
     {targetBone {boneName=spine_03}}
     {targetBone {boneName=spine_01}}
     {targetBone {boneName=pelvis}}
    ]
    constraintWeights [1,1,1,1,1]
   }
   Save cached pose 'dyn cache'(Pose=Component To Local(ComponentPose=@n31)) {cacheName=dyn cache}
   n32 = Get Body Mesh() -> (Body Mesh:Skeletal Mesh Component Object Reference)
   n33 = Copy Pose From Mesh(SourceMeshComponent=Get Body Mesh()) {bCopyCurves=true, bUseAttachedParent=true}
   n34 = Layered blend per bone(
    BasePose=Copy Pose From Mesh() {bCopyCurves=true, bUseAttachedParent=true}
    BlendPoses_0=Copy Pose From Mesh(SourceMeshComponent=Get Body Mesh()) {bCopyCurves=true, bUseAttachedParent=true}
   ) {layerSetup [{branchFilters [{boneName=upperarm_l}, {boneName=upperarm_r}, {boneName=lowerarm_l}, {boneName=lowerarm_r}]}]}
   n35 = Control Rig(Source=Layered blend per bone(
     BasePose=Layered blend per bone(BasePose=@n34, BlendPoses_0=@n33) {layerSetup [{branchFilters [{boneName=calf_r}, {boneName=calf_l}]}]}
     BlendPoses_0=Copy Pose From Mesh(SourceMeshComponent=@n32) {bCopyCurves=true, bUseAttachedParent=true}
    ) {layerSetup [{branchFilters [{boneName=lowerarm_r}, {boneName=lowerarm_l}]}]}) {controlRigAssetReference {blueprintRigClass=&/F_MED_QuicheLorraineCrisp_Hair_FaceAcc_CtrlRig.F_MED_QuicheLorraineCrisp_Hair_FaceAcc_CtrlRig_C}}
   n36 = "Constraint - dyn_untucked_front_shoulder_base (PPPP)"!(
    ComponentPose=Local To Component(LocalPose=@n35)
    Alpha=1
    ConstraintWeights_0=0.0515
    ConstraintWeights_1:float=0.605
    ConstraintWeights_2:float=0.3329
    ConstraintWeights_3:float=0.0106
   ) -> (Pose) {
    boneToModify {boneName=dyn_untucked_front_shoulder_base}
    constraintSetup [
     {targetBone {boneName=spine_04}}
     {targetBone {boneName=spine_05}}
     {targetBone {boneName=clavicle_r}}
     {targetBone {boneName=upperarm_twist_01_r}}
    ]
    constraintWeights [1,1,1,1]
   }
   Save cached pose 'CtrlRig_Cache'(Pose=Component To Local(ComponentPose=@n36)) {cacheName=CtrlRig_Cache}
  ]}
 EventGraph {nodes [
   n1 = Event BlueprintUpdateAnimation()
   > n2 = "Gravity Override State Function (Parameter Asset) (Nativized)"(
    CharacterPartAnimInstance=Self-Reference()
    GameTimeInSeconds=@n1.DeltaTimeX
    DebugJointName=neck_02
    Parameters=&/F_MED_QuicheLorraineCrisp_Parameters1
   )
   > Set Gravity Override(Gravity Override=@n2)
   n3 = GetOwnerSkeletalMeshComponent()
   > Set CustomEmoteLime(CustomEmoteLime:double=GetCurveValue(self=GetAnimInstance(self=@n3), CurveName=QuicheLorraineLimeEmote)) -> (Output_Get:double)
   Event OnCharacterCustomizationLoaded()
   > n4 = Sequence!(execute) -> (then_0, then_1, then_2:Exec)
   n5 = CheckForBackbling()
   > Set Backbling(Backbling:double="To Float (Boolean)"(InBool=@n5)) -> (Output_Get:double)
   > n6 = Branch(Condition=@n5)
   > Set Normal Alpha(Normal Alpha:double=0) -> (Output_Get:double)
   Set Normal Alpha(execute=@n6.else, Normal Alpha=1)
   n7 = Get bIsPlayingEmote()
   Set IdleFE(execute=@n4.then_1, IdleFE:bool=AND Boolean(A=Get bIsInFrontEnd(), B=NOT Boolean(A=@n7))) -> (Output_Get:bool)
   n8 = CheckForBackbling()
   Set isWearingBackbling(
    execute=@n4.then_2
    isWearingBackbling=AND Boolean!(A=@n8, B=NOT Boolean(A=@n8.bIsBackblingACape), C:bool=NOT Boolean(A=@n8.bIsBackblingLong)) -> (ReturnValue)
   )
   > Set Is Wearing CapeSKy(Is Wearing CapeSKy:bool=AND Boolean!(A=Get bIsSkyDiving(), B=OR Boolean(A=@n8.bIsBackblingACape, B=@n8.bIsBackblingLong)) -> (ReturnValue)) -> (Output_Get:bool)
   Event OnBeginPlayingEmote() {is=title K2Node_Event}
   Event OnEndPlayingEmote() {is=title K2Node_Event}
   Set IdleFE(IdleFE=false)
   Set IdleFE(IdleFE=false)
  ]}
]
variables [
 Backbling {type=number}
 Body Mesh {type=SkeletalMeshComponent}
 CustomEmoteLime {type=number}
 Gravity Override {type=Vector, default=<0 0 -980>}
 IdleFE {type=boolean}
 Is Wearing CapeSKy {type=boolean}
 isWearingBackbling {type=boolean}
 NewVar {type=AnimInstance, default=null}
 Normal Alpha {type=number, default=1}
 Skel_Mesh {type=SkeletalMeshComponent}
]
